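# Shared wheel cosmetic AnimBlueprint that spins a wheel's spinner, targeting the Dragon wheel skeleton.
# Starts from the mesh-space ref pose, adds the SpinnerSpin rotator to spinner_jnt through an additive Modify Bone node, then converts to local space for output.
# SpinnerSpin is not declared in this Blueprint, and the event graph holds only unconnected stub nodes, so no spin value is computed here.
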AnimBlueprint /VehicleCosmetics/Wheels/_SharedAnimations/Wheel_Spinner_AnimBP
target_skeleton=/VehicleCosmetics/Wheels/Dragon/Meshes/WHEEL_Dragon_Skeleton
graphs [
 AnimGraph {nodes [
   n1 = Mesh Space Ref Pose!() -> (ComponentPose:Component Space Pose Link Structure) {is=title AnimGraphNode_MeshRefPose}
   n2 = Get SpinnerSpin() -> (SpinnerSpin:Rotator)
   Output Pose(Result=Component To Local(ComponentPose="Transform (Modify) Bone - Bone: spinner_jnt"(ComponentPose=@n1, Alpha=1, Translation="0,0,0", Rotation=@n2, Scale="1,1,1") {rotationMode=BMM_Additive, alphaBoolBlend {blendOption=Linear}, boneToModify {boneName=spinner_jnt}})) {layerGroup=None}
  ]}
 EventGraph {nodes [
   TryGetPawnOwner()
   Event BlueprintUpdateAnimation()
  ]}
]
variables []
